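# Masked, lit chrome material for Niagara mesh particles in the Fractured Beacon aura cosmetic, faking an iridescent metallic reflection.
# A masked world reflection vector, tiled and offset through MF_PanTileOffset, samples a thin-film quartz texture that is desaturated and sharpened by a power.
# That result feeds a Fresnel add, a tintable rainbow blend and an emissive adjustment function.
# Particle dynamic parameters drive reflection tiling and glow amount.
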Material /SparksCosmetics/Items/Auras/Assets/NS_Aura_FracturedBeacon/M_FracturedBeacon_Chrome
graph {
 properties {bAutomaticallySetUsageInEditor=false, bUsedWithNiagaraMeshParticles=true, blendMode=BLEND_Masked, usedShadingModels=Default Lit}
 nodes [
  n1 = ReflectionVectorWS()
  n2 = ComponentMask(None=@n1) {b=true}
  n3 = DynamicParameter() {defaultValue=<rgba 0.25 0 0.55 1>, paramNames [ReflectionScale, GlowAmount, ReflectionOffset, Param4]}
  n4 = Desaturation(
   None=TextureSample(UVs=MaterialFunctionCall!(UV=@n2, Tile U=@n3.ReflectionScale, Tile V=@n3.ReflectionScale, Pan U, Pan V, Offset U=@n3.ReflectionScale, Offset V, Time) {materialFunction=/Game/Effects/Fort_Effects/Materials/_Functions/MF_PanTileOffset}.UV's) {texture=/Game/Gadgets/Assets/VinderTech_GliderChute/Glider_Quartz/Textures/T_ThinFilm_Quartz_Color}.RGB
   Fraction=ScalarParameter() {defaultValue=0.75, parameterName=Desaturation}
  )
  n5 = Power(Base=@n4, Exp=ScalarParameter() {defaultValue=2, parameterName=Reflection Exp}) {constExponent=1.5}
  n6 = Constant() {r=4}
  Add(A=@n5, B=SmoothStep(Value=Fresnel(ExponentIn=@n6) {baseReflectFraction=0.004}) {constMax=0.1})
  ComponentMask(None=@n1) {b=true, g=true}
  Constant() {r=1}
  Constant() {r=0.2}
  Constant() {r=1}
  Constant() {r=0.2}
  LinearInterpolate(
   A=VectorParameter() {parameterName=Base Color, defaultValue=<rgba 1 1 1 1>}.RGB
   B=@n4
   Alpha=ScalarParameter() {defaultValue=1, parameterName=RainbowStrength}
  )
  n7 = ScalarParameter() {defaultValue=0.35, parameterName=EmissiveContribution}
  MaterialFunctionCall(Emissive=Multiply(
    A=Multiply(A=@n5, B=@n7)
    B=DynamicParameter() {defaultValue=<rgba 0.25 0 0.55 1>, paramNames [ReflectionScale, GlowAmount, ReflectionOffset, Param4]}.GlowAmount
   )) {materialFunction=&MF_FX_EmissiveAdjustments}
  ScalarParameter() {defaultValue=0.25, parameterName=Reflection Scale}
  ScalarParameter() {defaultValue=0.55, parameterName=ReflectionOffset}
  TextureSample(UVs=MaterialFunctionCall(UV=ComponentMask(None=@n1) {g=true, r=true}, Tile U=@n3.ReflectionScale, Tile V=@n3.ReflectionScale, Offset U=@n3.ReflectionScale) {materialFunction=&MF_PanTileOffset}.UV's) {texture=&T_ThinFilm_Quartz_Color}
 ]
}
